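# Translucent UI material for the map rotation entry in the FN1 Shooters matchmaking settings, showing a map image on a styled backplate.
# It fits the Texture parameter to the right using widget and texture aspect ratios, then cuts it against a blue-to-green gradient with a slanted mask set by Angle.
# A horizontal ramp partially blends that composite over the CoreUI blurred backplate.
# ShowMap blends between a flat dark color and the result.
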Material /MatchmakingSettingsFN1Shooters/MatchmakingSettings/Entries/MapRotation/Materials/M_UI_FN1_MatchmakingSettings_MapRotation
graph {
 properties {bAutomaticallySetUsageInEditor=false, blendMode=BLEND_Translucent, materialDomain=MD_UI, usedShadingModels=Default Lit}
 nodes [
  n1 = OneMinus()
  AppendVector(A=OneMinus(None=Multiply(A=@n1)))
  n2 = LinearInterpolate(
   A=Constant3Vector() {constant=<rgba 0.0015 0.0185 0.0612 1>}
   B=Constant3Vector() {constant=<rgba 0.0782 0.9911 0.1195 1>}
   Alpha=MaterialFunctionCall(GradientStart=Constant() {r=0.1}, GradientEnd=Constant() {r=2}, RotationAngle=Constant() {r=-0.05}, GradientOffset=Constant() {r=-0.1}) {materialFunction=&MF_UI_LinearGradient}.Result
  )
  n3 = TextureCoordinate()
  n4 = ComponentMask(None=@n3) {r=true}
  n5 = MaterialFunctionCall() {materialFunction=&GetUserInterfaceUV}
  n6 = TextureObjectParameter() {parameterName=Texture}
  n7 = TextureProperty(Tex=@n6)
  n8 = ComponentMask(None=@n7) {r=true}
  n9 = ComponentMask(None=@n7) {g=true}
  n10 = Divide(A=Divide(A=ComponentMask(None=@n5.Pixel Size) {r=true}, B=Divide(A=@n8, B=@n9)) {constB=0.5}, B=ComponentMask(None=@n5.Pixel Size) {g=true})
  n11 = Multiply(A=OneMinus(None=@n4), B=@n10)
  n12 = ComponentMask(None=@n3) {g=true}
  n13 = NamedRerouteDeclaration(None=AppendVector(A=OneMinus(None=@n11), B=@n12)) {name=AspectRatio_Right, nodeColor=<rgba 0.3828 0 1 1>}
  n14 = TextureSample(UVs=AppendVector(A=ComponentMask(None=@n13) {r=true}, B=ComponentMask(None=@n13) {g=true}), Tex=@n6) {samplerSource=SSM_Clamp_WorldGroupSettings}
  n15 = MaterialFunctionCall(UVs=@n13, "Rotation Angle (0-1)"=ScalarParameter() {defaultValue=-0.03, parameterName=Angle}) {materialFunction=&CustomRotator}
  n16 = NamedRerouteDeclaration(None=SmoothStep(Value=ComponentMask(None=@n15.Rotated Values) {r=true}) {constMax=0.104, constMin=0.1}) {name=Mask, nodeColor=<rgba 0.9688 0 1 1>}
  n17 = SmoothStep(Value=ComponentMask(None=MaterialFunctionCall() {materialFunction=&GetUserInterfaceUV}.Normalized UV) {r=true})
  LinearInterpolate(
   A=Constant3Vector() {constant=<rgba 0.0048 0.004 0.0116 1>}
   B=LinearInterpolate(
    A=MaterialFunctionCall!(Post Buffer Blur) {materialFunction=/FortUILibrary/MaterialFunctions/CoreUI/MF_UI_CoreUI_BackplateC}.Result
    B=LinearInterpolate(A=@n2, B=@n14.RGB, Alpha=@n16)
    Alpha=LinearInterpolate(Alpha=@n17) {constB=0.6}
   )
   Alpha=ScalarParameter() {defaultValue=1, parameterName=ShowMap}
  )
  NamedRerouteDeclaration(None=AppendVector(A=Multiply(A=@n4, B=@n10), B=@n12)) {name=AspectRatio_Left, nodeColor=<rgba 0.3828 0 1 1>}
 ]
}
