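# Map rotation entry widget for matchmaking settings, showing either the current or next map over a material background.
# PreConstruct uses IsCurrentMap to choose the heading text, drive the background material's ShowMap parameter and set the minimum height via a spacer, shorter for next map.
# It also broadcasts the compact-UI state.
# The layout pairs the heading and height-capped map name with a named slot for extra map info.
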WidgetBlueprint /MatchmakingSettingsFN1Shooters/MatchmakingSettings/Entries/MapRotation/WBP_FN1_MatchmakingSettings_MapRotationWidget
graphs [EventGraph {nodes [
   Event PreConstruct()
   n1 = Get IsCurrentMap() -> (IsCurrentMap:bool)
   > "SetText (Text)"(
    self=Get Txt_Map() -> (Txt_Map:Common Text Object Reference)
    InText=Select(Option 0:Text="LOCTEXT(\"Next Map:\")", Option 1:Text="LOCTEXT(\"Current Map:\")", Index=@n1) -> (ReturnValue:Text)
   )
   > n2 = GetDynamicMaterial(self=Get Img_Map() -> (Img_Map:Image Object Reference))
   > n3 = Set MID_Background(MID_Background:Material Instance Dynamic Object Reference=@n2) -> (Output_Get:Material Instance Dynamic Object Reference)
   > SetScalarParameterValue(self=@n3, ParameterName=ShowMap, Value="To Float (Boolean)"(InBool=@n1))
   > SetSize!(
    execute
    self:Spacer Object Reference=Get MinSize() -> (MinSize:Spacer Object Reference)
    InSize_X:double=536
    InSize_Y:double=Select(Option 0:double=80, Option 1:double=112, Index=@n1) -> (ReturnValue:double)
   ) -> (then)
   > Set with Broadcast IsUsingCompactUI!(execute, IsUsingCompactUI:bool=IsUsingCompactUI()) -> (then, Output_Get:bool)
  ]}]
variables [
 IsCurrentMap {type=boolean, default=true}
 IsUsingCompactUI {type=boolean}
 MID_Background {type=MaterialInstanceDynamic}
]
widgets {
 info {inheritedWidgetCount=0, parentClass=&UserWidget, rootWidgetClass=&Overlay, widgetCount=9}
 widgets [
  Overlay_0 {widgetClassPath=&Overlay, widget=Overlay_0}
  MinSize {
   widgetClassPath=&Spacer
   bIsVariable=true
   parent=Overlay_0
   widget=MinSize
   properties {size=<536 112>}
   slot {class=OverlaySlot, properties {horizontalAlignment=HAlign_Fill, verticalAlignment=VAlign_Fill}}
  }
  GridPanel_0 {
   widgetClassPath=&GridPanel
   parent=Overlay_0
   widget=GridPanel_0
   properties {bOverride_Cursor=true, columnFill [0,1], rowFill [1]}
   slot {class=OverlaySlot, properties {horizontalAlignment=HAlign_Fill, verticalAlignment=VAlign_Fill}}
  }
  Img_Map {
   widgetClassPath=&Image
   bIsVariable=true
   parent=GridPanel_0
   widget=Img_Map
   properties {
    visibility=HitTestInvisible
    brush {
     bIsDynamicallyLoaded=false
     drawAs=Image
     imageType=FullColor
     mirroring=NoMirror
     resourceName=None
     resourceObject=/MatchmakingSettingsFN1Shooters/MatchmakingSettings/Entries/MapRotation/MaterialInstances/MI_FN1_MapRotation
     tiling=NoTile
     imageSize=<0 0>
     margin=<ltrb 0 0 0 0>
     outlineSettings {
      bUseBrushTransparency=false
      roundingType=HalfHeightRadius
      width=0
      color {colorUseRule=UseColor_Specified, specifiedColor=<rgba 0 0 0 0>}
      cornerRadii=<0 0 0 0>
     }
     tintColor {colorUseRule=UseColor_Specified, specifiedColor=<rgba 1 1 1 1>}
     uVRegion {bIsValid=false, max=<0 0>, min=<0 0>}
    }
   }
   slot {class=GridSlot, properties {columnSpan=2}}
  }
  VerticalBox_0 {
   widgetClassPath=&VerticalBox
   parent=GridPanel_0
   widget=VerticalBox_0
   slot {class=GridSlot, properties {horizontalAlignment=HAlign_Left, verticalAlignment=VAlign_Center, padding=<ltrb 30 16 0 16>}}
  }
  Txt_Map {
   widgetClassPath=&CommonTextBlock
   bIsVariable=true
   parent=VerticalBox_0
   widget=Txt_Map
   properties {
    autoWrapText=true
    minDesiredWidth=260
    mobileFontSizeMultiplier=1.2
    scrollStyle=&TextScrollStyle-Base_C
    colorAndOpacity {colorUseRule=UseColor_Specified, specifiedColor=<rgba 1 1 1 0.5>}
    font {
     bForceMonospaced=false
     bMaterialIsStencil=false
     fontMaterial=null
     fontObject=&HeadingNow
     letterSpacing=0
     monospacedWidth=1
     size=15
     skewAmount=0
     typefaceFontName=75 Medium
     outlineSettings {
      bApplyOutlineToDropShadows=false
      bMiteredCorners=false
      bSeparateFillAlpha=false
      outlineMaterial=null
      outlineSize=0
      outlineColor=<rgba 0 0 0 1>
     }
    }
    shadowOffset=<0 0>
   }
   slot {class=VerticalBoxSlot}
  }
  SZ_MapName {
   widgetClassPath=&SizeBox
   bIsVariable=true
   parent=VerticalBox_0
   widget=SZ_MapName
   properties {bOverride_MaxDesiredHeight=true, maxDesiredHeight=62}
   slot {class=VerticalBoxSlot}
  }
  Txt_MapName {
   widgetClassPath=&CommonTextBlock
   bIsVariable=true
   parent=SZ_MapName
   widget=Txt_MapName
   properties {
    autoWrapText=true
    minDesiredWidth=260
    mobileFontSizeMultiplier=1.2
    scrollOrientation=Orient_Vertical
    scrollStyle=&TextScrollStyle-Base_C
    text=$MapName
    wrapTextAt=230
    font {
     bForceMonospaced=false
     bMaterialIsStencil=false
     fontMaterial=null
     fontObject=&HeadingNow
     letterSpacing=0
     monospacedWidth=1
     size=18
     skewAmount=0
     typefaceFontName=86 Bold
     outlineSettings {
      bApplyOutlineToDropShadows=false
      bMiteredCorners=false
      bSeparateFillAlpha=false
      outlineMaterial=null
      outlineSize=0
      outlineColor=<rgba 0 0 0 1>
     }
    }
    shadowOffset=<0 0>
   }
   slot {class=SizeBoxSlot}
  }
  NamedSlot_MapInfo {
   widgetClassPath=&NamedSlot
   bIsVariable=true
   parent=GridPanel_0
   widget=NamedSlot_MapInfo
   slot {class=GridSlot, properties {column=1}}
  }
 ]
}
